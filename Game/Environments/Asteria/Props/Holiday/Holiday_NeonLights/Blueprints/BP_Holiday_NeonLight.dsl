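; BP_Holiday_NeonLight is a configurable holiday neon light prop based on StaticMeshActor, set up entirely in its construction script.
; The script picks the static mesh from the Meshes array using the Mesh selection.
; With Override Color on, it creates a dynamic material instance on slot 0 and sets Color and Brightness from Light Color and Emissive Intensity.
; Otherwise it applies a preset material chosen by Color Preset; the event graph events are empty.
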(event EventBeginPlay)

(event Collision|EventActorBeginOverlap
  (params (OtherActor :type Actor%20Object%20Reference)))

(event EventTick
  (params (DeltaSeconds :type Float%20%28single-precision%29)))

(fn ConstructionScript
  (Components|StaticMesh|SetStaticMesh (Variables|StaticMeshActor|GetStaticMeshComponent) (Utilities|Array|Get(acopy) (Variables|NeonLight|GetMeshes) (Variables|NeonLight|GetMesh)))
  (if (Variables|NeonLight|GetOverride%20Color)
    (bind _returnvalue (Rendering|Material|CreateDynamicMaterialInstance (Variables|StaticMeshActor|GetStaticMeshComponent) 0 0 "None" :!class /Script/Engine.PrimitiveComponent))
    (Rendering|Material|SetVectorParameterValue _returnvalue "Color" (Variables|NeonLight|GetLightColor) :!class /Script/Engine.MaterialInstanceDynamic)
    (Rendering|Material|SetScalarParameterValue _returnvalue "Brightness" (Variables|NeonLight|GetEmissiveIntensity) :!class /Script/Engine.MaterialInstanceDynamic)
    (else
      (Rendering|Material|SetMaterial (Variables|StaticMeshActor|GetStaticMeshComponent) 0 (Utilities|Array|Get(acopy) (Variables|NeonLight|GetColorPresetMaterials) (Variables|NeonLight|GetColorPreset))))))
